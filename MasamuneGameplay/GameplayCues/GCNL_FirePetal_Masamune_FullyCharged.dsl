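; Looping gameplay cue notify for Masamune's fully charged Fire Petal effect, with a color variant switch that a note attributes to the Nana Eternal boss.
; On looping start, it sets the first particle component's "Color Variant" integer from whether the target's root component display name contains "Smash".
; This name-based check is an acknowledged workaround, and a debug print of that name remains.
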(event EventOnLoopingStart(EitherVFX-preferthisversion)
  (params (MyTarget :type Actor%20Object%20Reference) (Parameters :type Gameplay%20Cue%20Parameters%20Structure%20%28by%20ref%29) (ParticleComponents :type Array%20of%20FXSystem%20Component%20Object%20References) (AudioComponents :type Array%20of%20Audio%20Component%20Object%20References))
  (bind _rootcomponent (Variables|/Script/Engine.Actor|GetRootComponent MyTarget))
  (bind _returnvalue (Utilities|GetDisplayName _rootcomponent))
  ; for Nana Eternal Boss.. I know it's not good way tho..
  (Effects|Components|ParticleSystem|SetIntParameter (Utilities|Array|Get(acopy) ParticleComponents) "Color Variant" (Math|Conversions|ToInteger(Boolean) (Utilities|String|Contains _returnvalue "Smash")))
  (Development|PrintString 0 _returnvalue true true "(R=0.000000,G=0.155049,B=1.000000,A=1.000000)" 11.0))

(fn ConstructionScript)
