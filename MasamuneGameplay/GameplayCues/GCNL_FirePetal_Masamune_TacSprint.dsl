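; Looping gameplay cue for the FirePetal Masamune tactical sprint that can hide its sprint VFX while the target player is cloaked.
; On loop start it stores the target pawn and, if the HideSprintEffects scalable float is above zero, starts async watchers for the cloak tag being added or removed.
; Those watchers hide or show the first two Niagara components and their children, and their task proxies are saved in variables.
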(event EventOnLoopingStart(EitherVFX-preferthisversion)
  (params (MyTarget :type Actor%20Object%20Reference) (Parameters :type Gameplay%20Cue%20Parameters%20Structure%20%28by%20ref%29) (ParticleComponents :type Array%20of%20FXSystem%20Component%20Object%20References) (AudioComponents :type Array%20of%20Audio%20Component%20Object%20References))
  (bind _asplayer_pawn_athena (Utilities|Casting|CastToPlayerPawn_Athena MyTarget))
  (Variables|Default|SetTargetActor _asplayer_pawn_athena)
  (if (> (ScalableFloat|GetValueasDouble (Variables|Default|GetHideSprintEffects)) 0)
    (bind _asynctaskproxy (Ability|Tasks|WaitGameplayTagAddToActor (Variables|Default|GetTargetActor) (Variables|Default|GetCloakTag))
      (:then
        (Variables|Default|SetTag%20Add%20Watcher _asynctaskproxy)
        (bind _asynctaskproxy_2 (Ability|Tasks|WaitGameplayTagRemoveFromActor (Variables|Default|GetTargetActor) (Variables|Default|GetCloakTag))
          (:then
            (Variables|Default|SetTag%20Remove%20Watcher _asynctaskproxy_2)
            (bind _result (ArrayMacros|SafeArrayGet ParticleComponents)
              (:%20%20
                (bind _result_2 (ArrayMacros|SafeArrayGet ParticleComponents 1)))))
          (:Removed
            (bind _asniagara_particle_system_component (Utilities|Casting|CastToNiagaraComponent _result))
            (Rendering|SetVisibility _asniagara_particle_system_component true true)
            (bind _asniagara_particle_system_component_2 (Utilities|Casting|CastToNiagaraComponent _result_2))
            (Rendering|SetVisibility _asniagara_particle_system_component_2 true true))))
      (:Added
        (bind _asniagara_particle_system_component_3 (Utilities|Casting|CastToNiagaraComponent _result))
        (Rendering|SetVisibility _asniagara_particle_system_component_3 false true)
        (bind _asniagara_particle_system_component_4 (Utilities|Casting|CastToNiagaraComponent _result_2))
        (Rendering|SetVisibility _asniagara_particle_system_component_4 false true)))))

(fn ConstructionScript)
